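# Physics asset for the Kevin Couture male medium head mesh, giving three dynamic "dyn_fragment" bones their physics bodies.
# Each fragment capsule is constrained to a tiny capsule on its _MSTR bone, which in turn is constrained to a box body on the head.
# All constraints leave swing and twist free, so their listed angle limits do not restrict rotation.
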PhysicsAsset /BRCosmetics/Characters/Player/Male/Medium/Heads/M_Med_KevinCouture_Head/Meshes/Male_Medium_Kevin_Couture_Head_Dyn
bodies {
 dyn_fragment_01 [{shapeType=Capsule, length=9.9688, radius=2, center=<-0.7035 1.2699 0.449>, rotation=<rot 0 0 30.0001>}]
 dyn_fragment_01_MSTR [{shapeType=Capsule, length=0.505, radius=0.505, rotation=<rot 0 -90 0>}]
 dyn_fragment_02 [{shapeType=Capsule, length=9.9688, radius=2, center=<0 -0.608 -0.7244>, rotation=<rot 0 0 -50>}]
 dyn_fragment_02_MSTR [{shapeType=Capsule, length=0.505, radius=0.505, rotation=<rot 0 -90 0>}]
 dyn_fragment_03 [{shapeType=Capsule, length=9.9688, radius=2, center=<0 0 0.9916>}]
 dyn_fragment_03_MSTR [{shapeType=Capsule, length=0.505, radius=0.505, rotation=<rot 0 -90 0>}]
 head [{
   shapeType=Box
   extentX=21.5
   extentY=21.5
   extentZ=21.5
   shapeName=head_box
   center=<6.9733 2.9364 -0.3352>
   rotation=<rot 3.6778 17.177 135>
  }]
}
constraints [
 {
  bone1Name=dyn_fragment_01
  bone2Name=dyn_fragment_01_MSTR
  swing1LimitDegrees=45
  swing1Motion=Free
  swing2LimitDegrees=45
  swing2Motion=Free
  twistLimitDegrees=45
  twistMotion=Free
 }
 {
  bone1Name=dyn_fragment_02
  bone2Name=dyn_fragment_02_MSTR
  swing1LimitDegrees=45
  swing1Motion=Free
  swing2LimitDegrees=45
  swing2Motion=Free
  twistLimitDegrees=45
  twistMotion=Free
 }
 {
  bone1Name=dyn_fragment_03
  bone2Name=dyn_fragment_03_MSTR
  swing1LimitDegrees=45
  swing1Motion=Free
  swing2LimitDegrees=45
  swing2Motion=Free
  twistLimitDegrees=45
  twistMotion=Free
 }
 {
  bone1Name=dyn_fragment_01_MSTR
  bone2Name=head
  swing1LimitDegrees=45
  swing1Motion=Free
  swing2LimitDegrees=45
  swing2Motion=Free
  twistLimitDegrees=45
  twistMotion=Free
 }
 {
  bone1Name=dyn_fragment_02_MSTR
  bone2Name=head
  swing1LimitDegrees=45
  swing1Motion=Free
  swing2LimitDegrees=45
  swing2Motion=Free
  twistLimitDegrees=45
  twistMotion=Free
 }
 {
  bone1Name=dyn_fragment_03_MSTR
  bone2Name=head
  swing1LimitDegrees=45
  swing1Motion=Free
  swing2LimitDegrees=45
  swing2Motion=Free
  twistLimitDegrees=45
  twistMotion=Free
 }
]
