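# Animation Blueprint for the Kevin Couture male medium head cosmetic, which makes three fragment bones float and adds secondary physics.
# Each update builds three offset vectors from game-time sine and cosine waves at differing frequencies and amplitudes.
# The pose copied from the attached parent mesh receives these as additive bone-space translations, then runs through a RigidBody node using the head's Dyn physics asset.
# A remapped pawn speed is computed but never used.
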AnimBlueprint /BRCosmetics/Characters/Player/Male/Medium/Heads/M_Med_KevinCouture_Head/Meshes/Male_Medium_Kevin_Couture_Head_AnimBP
target_skeleton=&/Male_Medium_Kevin_Couture_Head_Skeleton
graphs [
 AnimGraph {nodes [
   n1 = "Transform (Modify) Bone - Bone: dyn_fragment_01_MSTR"(
    ComponentPose=Local To Component(LocalPose=Copy Pose From Mesh() {bUseAttachedParent=true})
    Alpha=1
    Translation=Get ShardFloatA() -> (ShardFloatA:Vector)
    Rotation="0,0,0"
    Scale="1,1,1"
   ) {
    translationMode=BMM_Additive
    translationSpace=BCS_BoneSpace
    alphaBoolBlend {blendOption=Linear}
    boneToModify {boneName=dyn_fragment_01_MSTR}
   }
   n2 = "Transform (Modify) Bone - Bone: dyn_fragment_02_MSTR"(ComponentPose=@n1, Alpha=1, Translation=Get ShardFloatB() -> (ShardFloatB:Vector), Rotation="0,0,0", Scale="1,1,1") {
    translationMode=BMM_Additive
    translationSpace=BCS_BoneSpace
    alphaBoolBlend {blendOption=Linear}
    boneToModify {boneName=dyn_fragment_02_MSTR}
   }
   n3 = "Transform (Modify) Bone - Bone: dyn_fragment_03_MSTR"(ComponentPose=@n2, Alpha=1, Translation=Get ShardFloatC() -> (ShardFloatC:Vector), Rotation="0,0,0", Scale="1,1,1") {
    translationMode=BMM_Additive
    translationSpace=BCS_BoneSpace
    alphaBoolBlend {blendOption=Linear}
    boneToModify {boneName=dyn_fragment_03_MSTR}
   }
   n4 = RigidBody!(ComponentPose=@n3, Alpha, OverrideWorldGravity:Vector="0,0,-980", ExternalForce) -> (Pose) {
    bClampLinearTranslationLimitToRefPose=true
    bOverrideWorldGravity=true
    overridePhysicsAsset=&/Male_Medium_Kevin_Couture_Head_Dyn
    simulationSpace=BaseBoneSpace
    alphaBoolBlend {blendOption=Linear}
    baseBoneRef {boneName=root}
    componentAppliedLinearAccClamp=<2000 3000 3000>
    componentLinearAccScale=<0.2 0.2 0.2>
    componentLinearVelScale=<0.2 0.2 0.2>
    overrideWorldGravity {z=-980}
    simSpaceSettings {maxAngularAcceleration=500, maxAngularVelocity=500, velocityScaleZ=0, worldAlpha=0.05, externalLinearDragV=<0.1 0.1 0.1>}
   }
   Output Pose(Result=Component To Local(ComponentPose=@n4)) {layerGroup=None}
  ]}
 EventGraph {nodes [
   Event BlueprintUpdateAnimation()
   > Set PawnSpeedRemap(PawnSpeedRemap:double=MapRangeClamped(Value=Get PawnSpeed2D(), InRangeA=0, InRangeB=549)) -> (Output_Get:double)
   n1 = GetGameTimeInSeconds()
   n2 = "Sin (Radians)"(A=float * float(A=float * float(A=1, B=1), B=@n1))
   n3 = "Sin (Radians)"(A=float * float(A=float * float(A=1, B=3), B=@n1))
   n4 = "Sin (Radians)"(A=float * float(A=float * float(A=1, B=1.5), B=@n1))
   > Set ShardFloatA(ShardFloatA:Vector=MakeVector(X=float * float(A=@n4, B=0.3), Y=float * float(A=@n2, B=0.56), Z=float * float(A=@n3, B=0.22))) -> (Output_Get:Vector)
   n5 = "Cos (Radians)"!(A:double=float * float(A=float * float(A=1, B=1.5), B=@n1)) -> (ReturnValue:double)
   n6 = "Cos (Radians)"(A=float * float(A=float * float(A=1, B=1.78), B=@n1))
   n7 = "Cos (Radians)"(A=float * float(A=float * float(A=1, B=1.15), B=@n1))
   > Set ShardFloatB(ShardFloatB:Vector=MakeVector(X=float * float(A=@n5, B=0.6), Y=float * float(A=@n6, B=0.725), Z=float * float(A=@n7, B=0.3))) -> (Output_Get:Vector)
   n8 = "Sin (Radians)"(A=float * float(A=float * float(A=1, B=1.5), B=@n1))
   n9 = "Sin (Radians)"(A=float * float(A=float * float(A=1, B=2), B=@n1))
   n10 = "Sin (Radians)"(A=float * float(A=float * float(A=1, B=1), B=@n1))
   > Set ShardFloatC(ShardFloatC:Vector=MakeVector(X=float * float(A=@n8, B=0.3), Y=float * float(A=@n10, B=0.6), Z=float * float(A=@n9, B=0.25))) -> (Output_Get:Vector)
   TryGetPawnOwner()
  ]}
]
variables [
 PawnSpeedRemap {type=number}
 ShardFloatA {type=Vector}
 ShardFloatB {type=Vector}
 ShardFloatC {type=Vector}
]
